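# Physics asset for the Elf Attack backpack cosmetic, giving its dynamic head and ear bones simple collision bodies and joints.
# Capsules cover the root, head and both ears, while a thin box covers the pack.
# Constraints chain root, pack and head, then branch to each ear.
# The head joint carries tight angle limits, while both ear joints set swing and twist free so the ears can flop.
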PhysicsAsset /BRCosmetics/Accessories/FORT_Backpacks/M_Med_Elf_Attack_Backpack/Meshes/M_MED_Elf_Attack_Backpack_Ear_Physics
bodies {
 dyn_head [{shapeType=Capsule, length=9.5694, radius=6.4558, rotation=<rot 0 0 13.3059>}]
 dyn_l_ear [{shapeType=Capsule, length=3.4752, radius=2.7208, center=<-1.8096 0 0.2945>}]
 dyn_r_ear [{shapeType=Capsule, length=3.4724, radius=2.7648, center=<1.2617 0 0.2029>}]
 pack [{shapeType=Box, extentX=68.0835, extentY=2.8975, extentZ=81.9283}]
 root [{shapeType=Capsule, length=3.464, radius=3.1087, rotation=<rot 90 90 0>}]
}
constraints [
 {bone1Name=pack, bone2Name=root, swing1LimitDegrees=45, swing2LimitDegrees=45, twistLimitDegrees=15}
 {bone1Name=dyn_head, bone2Name=pack, swing1LimitDegrees=5, swing2LimitDegrees=5, twistLimitDegrees=5}
 {
  bone1Name=dyn_l_ear
  bone2Name=dyn_head
  swing1LimitDegrees=45
  swing1Motion=Free
  swing2LimitDegrees=45
  swing2Motion=Free
  twistLimitDegrees=45
  twistMotion=Free
 }
 {
  bone1Name=dyn_r_ear
  bone2Name=dyn_head
  swing1LimitDegrees=45
  swing1Motion=Free
  swing2LimitDegrees=45
  swing2Motion=Free
  twistLimitDegrees=45
  twistMotion=Free
 }
]
