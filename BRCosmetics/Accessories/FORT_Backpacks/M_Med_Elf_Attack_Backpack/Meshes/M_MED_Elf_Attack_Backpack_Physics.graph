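# Physics asset for the Elf Attack backpack cosmetic, defining capsule bodies and joint limits for the elf's dyn_ bones: head, ears, arms and hands.
# The head hangs off root with tight limits, and both arms and ears attach to the head.
# Arms swing widely at the head, elbows and wrists are stiffer, fingers and thumbs are looser, and ears barely move.
# Root carries a small capsule plus a large offset one.
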PhysicsAsset /BRCosmetics/Accessories/FORT_Backpacks/M_Med_Elf_Attack_Backpack/Meshes/M_MED_Elf_Attack_Backpack_Physics
bodies {
 dyn_head [{shapeType=Capsule, length=2.4662, radius=9.6942, center=<0 -3.7362 -3.4752>, rotation=<rot -90 0 0>}]
 dyn_l_arm [{shapeType=Capsule, length=4.7505, radius=2.4141, center=<4.7105 -0.0001 0>, rotation=<rot 90 0 0>}]
 dyn_l_ear [{shapeType=Capsule, length=3.4752, radius=4.8232, shapeName=dyn_l_ear_capsule, center=<-3.0541 0 0>}]
 dyn_l_elbow [{shapeType=Capsule, length=3.5703, radius=1.9116, center=<3.6957 0.0001 0>, rotation=<rot 90 0 0>}]
 dyn_l_fingers [{
   shapeType=Capsule
   length=2.5126
   radius=1.5894
   shapeName=dyn_l_fingers_capsule
   center=<1.6828 0.5527 -0.0937>
   rotation=<rot 0 0.5517 -79.3678>
  }]
 dyn_l_thumb [{
   shapeType=Capsule
   length=1.5526
   radius=1.251
   shapeName=dyn_l_thumb_capsule
   center=<0.9292 0.0451 -0.1806>
   rotation=<rot 0 -69.9262 -60.9084>
  }]
 dyn_l_wrist [{shapeType=Capsule, length=5.337, radius=6.3413, shapeName=dyn_l_wrist_capsule, center=<0 -2.2362 1.3329>, rotation=<rot 0 0 24.9107>}]
 dyn_r_arm [{shapeType=Capsule, length=5.9836, radius=2.2792, center=<5.1399 0.0001 0>, rotation=<rot 90 0 0>}]
 dyn_r_ear [{shapeType=Capsule, length=3.4724, radius=3.677, shapeName=dyn_r_ear_capsule, center=<3.0224 0 0>}]
 dyn_r_elbow [{shapeType=Capsule, length=4.7411, radius=1.6134, center=<3.9515 0 0>, rotation=<rot 90 0 0>}]
 dyn_r_fingers [{
   shapeType=Capsule
   length=2.8033
   radius=1.9407
   shapeName=dyn_r_fingers_capsule
   center=<-0.58 -0.1467 0.9504>
   rotation=<rot 0 -162.6885 -57.7455>
  }]
 dyn_r_thumb [{
   shapeType=Capsule
   length=1.4233
   radius=0.926
   shapeName=dyn_r_thumb_capsule
   center=<-0.3001 1.0665 -0.0517>
   rotation=<rot 0 -143.1924 -83.0626>
  }]
 dyn_r_wrist [{
   shapeType=Capsule
   length=3.6336
   radius=7.2378
   shapeName=dyn_r_wrist_capsule
   center=<2.1328 0.0966 0.069>
   rotation=<rot 10.0393 0 0.0002>
  }]
 root [
  {shapeType=Capsule, length=3.464, radius=3.1087, rotation=<rot 90 90 0>}
  {shapeType=Capsule, length=59.8598, radius=21.6473, shapeName=root_capsule, center=<0 19.1046 0>, rotation=<rot 0 0 9.8109>}
 ]
}
constraints [
 {bone1Name=dyn_head, bone2Name=root, swing1LimitDegrees=5, swing2LimitDegrees=5, twistLimitDegrees=5}
 {bone1Name=dyn_l_elbow, bone2Name=dyn_l_arm, swing1LimitDegrees=15, swing2LimitDegrees=15, twistLimitDegrees=5}
 {bone1Name=dyn_l_wrist, bone2Name=dyn_l_elbow, swing1LimitDegrees=15, swing2LimitDegrees=15, twistLimitDegrees=5}
 {bone1Name=dyn_l_fingers, bone2Name=dyn_l_wrist, swing1LimitDegrees=45, swing2LimitDegrees=45, twistLimitDegrees=45}
 {bone1Name=dyn_l_thumb, bone2Name=dyn_l_wrist, swing1LimitDegrees=45, swing2LimitDegrees=45, twistLimitDegrees=45}
 {bone1Name=dyn_l_arm, bone2Name=dyn_head, swing1LimitDegrees=60, swing2LimitDegrees=60, twistLimitDegrees=5}
 {bone1Name=dyn_r_arm, bone2Name=dyn_head, swing1LimitDegrees=60, swing2LimitDegrees=60, twistLimitDegrees=5}
 {bone1Name=dyn_r_elbow, bone2Name=dyn_r_arm, swing1LimitDegrees=15, swing2LimitDegrees=15, twistLimitDegrees=5}
 {bone1Name=dyn_r_wrist, bone2Name=dyn_r_elbow, swing1LimitDegrees=15, swing2LimitDegrees=15, twistLimitDegrees=5}
 {bone1Name=dyn_r_fingers, bone2Name=dyn_r_wrist, swing1LimitDegrees=45, swing2LimitDegrees=45, twistLimitDegrees=45}
 {bone1Name=dyn_r_thumb, bone2Name=dyn_r_wrist, swing1LimitDegrees=45, swing2LimitDegrees=45, twistLimitDegrees=45}
 {bone1Name=dyn_r_ear, bone2Name=dyn_head, swing1LimitDegrees=10, swing2LimitDegrees=1, twistLimitDegrees=1}
 {bone1Name=dyn_l_ear, bone2Name=dyn_head, swing1LimitDegrees=10, swing2LimitDegrees=1, twistLimitDegrees=1}
]
