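# Animation Blueprint for the Elf Attack backpack cosmetic, adding physics-driven secondary motion to the elf's body and face.
# RigidBody simulation on the reference pose feeds bool blends choosing default, jumping/falling or crouching poses, the latter two using additive arm-bone offsets.
# Chained AnimDynamics nodes then jiggle hair, nose, jaw, eyes and hat.
# Ear physics and hat-root nodes are unconnected, and the sine-driven clap variables go unread.
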AnimBlueprint /BRCosmetics/Accessories/FORT_Backpacks/M_Med_Elf_Attack_Backpack/Meshes/M_MED_Elf_Attack_AnimBP
target_skeleton=&/M_MEM_MED_Elf_Attack_Backpack_Skeleton
graphs [
 AnimGraph {nodes [
   Anim Dynamics - Bone: dyn_hat(Alpha=1, ExternalForce="0,0,0") {
    bGravityOverrideInSimSpace=true
    gravityScale=0.2
    alphaBoolBlend {blendOption=Linear}
    boundBone {boneName=dyn_hat}
    physicsBodyDefinitions [{
      sphereCollisionRadius=0
      boundBone {boneName=dyn_hat}
      boxExtents=<10 10 5>
      constraintSetup {angularLimitsMax=<0 0 0>, angularLimitsMin=<0 0 0>, angularTarget=<0 0 0>}
      localJointOffset=<0 0 -1>
     }]
   }
   n1 = Anim Dynamics - Bone: dyn_hat(Alpha=0.15, ExternalForce="0,0,0") {
    alpha=0.9
    bGravityOverrideInSimSpace=true
    bOverrideLinearDamping=true
    gravityScale=0.2
    linearDampingOverride=0.9999
    numSolverIterationsPostUpdate=4
    numSolverIterationsPreUpdate=16
    alphaBoolBlend {blendOption=Linear}
    boundBone {boneName=dyn_hat}
    componentLinearAccScale=<0.25 0.25 0.25>
    physicsBodyDefinitions [{
      sphereCollisionRadius=0
      boundBone {boneName=dyn_hat}
      boxExtents=<15 15 10>
      constraintSetup {
       angularTargetAxis=AxisY
       twistAxis=AxisZ
       angularLimitsMax=<0 0 0>
       angularLimitsMin=<0 0 0>
       angularTarget=<0 1 0>
       linearAxesMax=<5 0.5 20>
       linearAxesMin=<-5 0 0>
      }
      localJointOffset=<0 0 -1>
     }]
    relativeSpaceBone {boneName=root}
   }
   n2 = Get bIsJumping()
   n3 = Get bIsFalling()
   Anim Dynamics - Bone: dyn_hatRoot(ComponentPose=@n1, Alpha=SelectFloat(A=1, B=0, bPickA=OR Boolean(A=@n3, B=@n2)), ExternalForce="0,0,0") {
    alpha=0.9
    angularDampingOverride=0.9999
    angularSpringConstant=50
    bAngularSpring=true
    bGravityOverrideInSimSpace=true
    bOverrideAngularDamping=true
    linearSpringConstant=150
    numSolverIterationsPostUpdate=4
    numSolverIterationsPreUpdate=16
    alphaBoolBlend {blendOption=Linear}
    boundBone {boneName=dyn_hatRoot}
    componentLinearAccScale=<1 1 1>
    physicsBodyDefinitions [{
      sphereCollisionRadius=0
      boundBone {boneName=dyn_hatRoot}
      boxExtents=<15 15 10>
      constraintSetup {twistAxis=AxisZ, angularLimitsMax=<5 5 5>, angularLimitsMin=<-5 -5 -5>, angularTarget=<0 1 0>}
      localJointOffset=<0 0 -1>
     }]
    relativeSpaceBone {boneName=root}
   }
   RigidBody(Alpha=0.6) {
    bClampLinearTranslationLimitToRefPose=true
    bOverrideWorldGravity=true
    overridePhysicsAsset=&/M_MED_Elf_Attack_Backpack_Ear_Physics
    simulationSpace=BaseBoneSpace
    alphaBoolBlend {blendOption=Linear}
    baseBoneRef {boneName=root}
    componentAppliedLinearAccClamp=<5000 5000 5000>
    componentLinearAccScale=<0.25 0.25 0.25>
    componentLinearVelScale=<0.25 0.25 0.25>
    overrideWorldGravity {z=-980}
   }
   n4 = RigidBody(ComponentPose=Local To Component(LocalPose=Local Space Ref Pose())) {
    bClampLinearTranslationLimitToRefPose=true
    bOverrideWorldGravity=true
    overridePhysicsAsset=&/M_MED_Elf_Attack_Backpack_Physics
    simulationSpace=BaseBoneSpace
    alphaBoolBlend {blendOption=Linear}
    baseBoneRef {boneName=pack}
    componentLinearAccScale=<0.3 0.3 0.3>
    componentLinearVelScale=<0.25 0.25 0.25>
    overrideWorldGravity {z=-980}
   }
   n5 = RigidBody(ComponentPose=Local To Component(LocalPose=Local Space Ref Pose())) {
    bClampLinearTranslationLimitToRefPose=true
    bOverrideWorldGravity=true
    overridePhysicsAsset=&M_MED_Elf_Attack_Backpack_Physics
    simulationSpace=BaseBoneSpace
    alphaBoolBlend {blendOption=Linear}
    baseBoneRef {boneName=pack}
    componentLinearAccScale=<0.3 0.3 0.3>
    componentLinearVelScale=<0.25 0.25 0.25>
    overrideWorldGravity {z=-980}
   }
   n6 = "Transform (Modify) Bone - Bone: dyn_l_arm"(ComponentPose=@n5, Alpha=1, Translation="0,0,0", Rotation="35,0,20", Scale="1,1,1") {
    rotationMode=BMM_Additive
    scaleMode=BMM_Additive
    translationMode=BMM_Additive
    alphaBoolBlend {blendOption=Linear}
    boneToModify {boneName=dyn_l_arm}
    rotation {pitch=35, roll=20}
   }
   n7 = "Transform (Modify) Bone - Bone: dyn_l_elbow"(ComponentPose=@n6, Alpha=1, Translation="0,0,0", Rotation="0,0.000,-10", Scale="1,1,1") {
    rotationMode=BMM_Additive
    scaleMode=BMM_Additive
    translationMode=BMM_Additive
    alphaBoolBlend {blendOption=Linear}
    boneToModify {boneName=dyn_l_elbow}
    rotation {roll=-10}
   }
   n8 = "Transform (Modify) Bone - Bone: dyn_l_wrist"(ComponentPose=@n7, Alpha=1, Translation="0,0,0", Rotation="-10,0,0", Scale="1,1,1") {
    rotationMode=BMM_Additive
    scaleMode=BMM_Additive
    translationMode=BMM_Additive
    alphaBoolBlend {blendOption=Linear}
    boneToModify {boneName=dyn_l_wrist}
    rotation {pitch=-10}
   }
   n9 = "Transform (Modify) Bone - Bone: dyn_r_arm"(ComponentPose=@n8, Alpha=1, Translation="0,0,0", Rotation="-20.000,0.000,0.000", Scale="1,1,1") {
    rotationMode=BMM_Additive
    scaleMode=BMM_Additive
    translationMode=BMM_Additive
    alphaBoolBlend {blendOption=Linear}
    boneToModify {boneName=dyn_r_arm}
    rotation {pitch=-20}
   }
   n10 = "Transform (Modify) Bone - Bone: dyn_r_elbow"(ComponentPose=@n9, Alpha=1, Translation="0,0,2", Rotation="0.000,-15,10.006", Scale="1,1,1") {
    rotationMode=BMM_Additive
    scaleMode=BMM_Additive
    translationMode=BMM_Additive
    alphaBoolBlend {blendOption=Linear}
    boneToModify {boneName=dyn_r_elbow}
    rotation {roll=10.006, yaw=-15}
    translation {z=2}
   }
   n11 = "Transform (Modify) Bone - Bone: dyn_r_wrist"(ComponentPose=@n10, Alpha=1, Translation="0,0,0", Rotation="0,0,0", Scale="1,1,1") {
    rotationMode=BMM_Additive
    scaleMode=BMM_Additive
    translationMode=BMM_Additive
    alphaBoolBlend {blendOption=Linear}
    boneToModify {boneName=dyn_r_wrist}
   }
   n12 = Blend Poses by bool(
    BlendPose_0=Component To Local(ComponentPose=@n11)
    BlendPose_1=Component To Local(ComponentPose=@n4)
    BlendTime_0=0.2
    BlendTime_1=0.2
    bActiveValue=OR Boolean(A=Get bIsJumping(), B=Get bIsFalling())
   ) {blendType=Linear}
   n13 = RigidBody(ComponentPose=Local To Component(LocalPose=Local Space Ref Pose())) {
    bClampLinearTranslationLimitToRefPose=true
    bOverrideWorldGravity=true
    overridePhysicsAsset=&M_MED_Elf_Attack_Backpack_Physics
    simulationSpace=BaseBoneSpace
    alphaBoolBlend {blendOption=Linear}
    baseBoneRef {boneName=pack}
    componentLinearAccScale=<0.3 0.3 0.3>
    componentLinearVelScale=<0.25 0.25 0.25>
    overrideWorldGravity {z=-980}
   }
   n14 = "Transform (Modify) Bone - Bone: dyn_l_arm"(ComponentPose=@n13, Alpha=1, Translation="0,0,0", Rotation="30,0,20", Scale="1,1,1") {
    rotationMode=BMM_Additive
    scaleMode=BMM_Additive
    translationMode=BMM_Additive
    alphaBoolBlend {blendOption=Linear}
    boneToModify {boneName=dyn_l_arm}
    rotation {pitch=30, roll=20}
   }
   n15 = "Transform (Modify) Bone - Bone: dyn_l_elbow"(ComponentPose=@n14, Alpha=1, Translation="0,0,0", Rotation="0,0.000,-10", Scale="1,1,1") {
    rotationMode=BMM_Additive
    scaleMode=BMM_Additive
    translationMode=BMM_Additive
    alphaBoolBlend {blendOption=Linear}
    boneToModify {boneName=dyn_l_elbow}
    rotation {roll=-10}
   }
   n16 = "Transform (Modify) Bone - Bone: dyn_l_wrist"(ComponentPose=@n15, Alpha=1, Translation="0,0,-1", Rotation="-20,0,0", Scale="1,1,1") {
    rotationMode=BMM_Additive
    scaleMode=BMM_Additive
    translationMode=BMM_Additive
    alphaBoolBlend {blendOption=Linear}
    boneToModify {boneName=dyn_l_wrist}
    rotation {pitch=-20}
    translation {z=-1}
   }
   n17 = "Transform (Modify) Bone - Bone: dyn_r_arm"(ComponentPose=@n16, Alpha=1, Translation="0,0,0", Rotation="-25,10,30", Scale="1,1,1") {
    rotationMode=BMM_Additive
    scaleMode=BMM_Additive
    translationMode=BMM_Additive
    alphaBoolBlend {blendOption=Linear}
    boneToModify {boneName=dyn_r_arm}
    rotation=<rot -25 10 30>
   }
   n18 = "Transform (Modify) Bone - Bone: dyn_r_elbow"(ComponentPose=@n17, Alpha=1, Translation="0,0,2", Rotation="0.000,-15,-20", Scale="1,1,1") {
    rotationMode=BMM_Additive
    scaleMode=BMM_Additive
    translationMode=BMM_Additive
    alphaBoolBlend {blendOption=Linear}
    boneToModify {boneName=dyn_r_elbow}
    rotation {roll=-20, yaw=-15}
    translation {z=2}
   }
   n19 = "Transform (Modify) Bone - Bone: dyn_r_wrist"(ComponentPose=@n18, Alpha=1, Translation="0,0,0", Rotation="0,-50,0", Scale="1,1,1") {
    rotationMode=BMM_Additive
    scaleMode=BMM_Additive
    translationMode=BMM_Additive
    alphaBoolBlend {blendOption=Linear}
    boneToModify {boneName=dyn_r_wrist}
    rotation {yaw=-50}
   }
   n20 = Component To Local(ComponentPose=@n19)
   n21 = Get bIsCrouching()
   n22 = Anim Dynamics - Bone: dyn_hair_01(
    ComponentPose=Local To Component(LocalPose=Blend Poses by bool(BlendPose_0=@n20, BlendPose_1=@n12, BlendTime_0=0.25, BlendTime_1=0.25, bActiveValue=@n21) {blendType=Linear})
    Alpha=1
   ) {
    angularDampingOverride=0.9999
    bGravityOverrideInSimSpace=true
    bOverrideAngularDamping=true
    gravityScale=0.25
    alphaBoolBlend {blendOption=Linear}
    boundBone {boneName=dyn_hair_01}
    physicsBodyDefinitions [{
      sphereCollisionRadius=0
      boundBone {boneName=dyn_hair_01}
      boxExtents=<15 15 5>
      constraintSetup {angularLimitsMax=<15 15 15>, angularLimitsMin=<-15 -15 -15>, angularTarget=<0 0 0>}
      localJointOffset=<0 0 -5>
     }]
   }
   n23 = Anim Dynamics - Bone: dyn_hair_02(ComponentPose=@n22, Alpha=1) {
    angularDampingOverride=0.9999
    bGravityOverrideInSimSpace=true
    bOverrideAngularDamping=true
    gravityScale=0.25
    alphaBoolBlend {blendOption=Linear}
    boundBone {boneName=dyn_hair_02}
    physicsBodyDefinitions [{
      sphereCollisionRadius=0
      boundBone {boneName=dyn_hair_02}
      boxExtents=<15 15 5>
      constraintSetup {angularLimitsMax=<15 15 15>, angularLimitsMin=<-15 -15 -15>, angularTarget=<0 0 0>}
      localJointOffset=<0 0 -1>
     }]
   }
   n24 = Anim Dynamics - Bone: dyn_hair_03(ComponentPose=@n23, Alpha=1, ExternalForce="0,0,0") {
    angularDampingOverride=0.9999
    bGravityOverrideInSimSpace=true
    bOverrideAngularDamping=true
    gravityScale=0.25
    alphaBoolBlend {blendOption=Linear}
    boundBone {boneName=dyn_hair_03}
    physicsBodyDefinitions [{
      sphereCollisionRadius=0
      boundBone {boneName=dyn_hair_03}
      boxExtents=<15 15 5>
      constraintSetup {angularLimitsMax=<15 15 15>, angularLimitsMin=<-15 -15 -15>, angularTarget=<0 0 0>}
      localJointOffset=<0 0 -2.5>
     }]
   }
   n25 = Anim Dynamics - Bone: dyn_nose(ComponentPose=@n24, Alpha=0.05, ExternalForce="0,0,0") {
    angularDampingOverride=0.9999
    angularSpringConstant=300
    bAngularSpring=true
    bGravityOverrideInSimSpace=true
    bLinearSpring=true
    bOverrideAngularDamping=true
    bOverrideLinearDamping=true
    gravityScale=0.2
    linearDampingOverride=0.9999
    linearSpringConstant=300
    alphaBoolBlend {blendOption=Linear}
    boundBone {boneName=dyn_nose}
    componentLinearAccScale=<1 1 1>
    physicsBodyDefinitions [{
      sphereCollisionRadius=0
      boundBone {boneName=dyn_nose}
      boxExtents=<15 15 5>
      constraintSetup {angularLimitsMax=<30 30 30>, angularLimitsMin=<-30 -30 -30>, angularTarget=<0 0 0>, linearAxesMax=<2 2 2>, linearAxesMin=<-2 -2 -2>}
      localJointOffset=<0 0 -1>
     }]
   }
   n26 = Anim Dynamics - Bone: dyn_jaw(ComponentPose=@n25, Alpha=0.15, ExternalForce="0,0,0") {
    angularDampingOverride=0.9999
    angularSpringConstant=300
    bAngularSpring=true
    bGravityOverrideInSimSpace=true
    bLinearSpring=true
    bOverrideAngularDamping=true
    bOverrideLinearDamping=true
    linearDampingOverride=0.9999
    linearSpringConstant=300
    alphaBoolBlend {blendOption=Linear}
    boundBone {boneName=dyn_jaw}
    componentLinearAccScale=<1 1 1>
    physicsBodyDefinitions [{
      sphereCollisionRadius=0
      boundBone {boneName=dyn_jaw}
      boxExtents=<15 15 5>
      constraintSetup {angularLimitsMax=<30 30 30>, angularLimitsMin=<-30 -30 -30>, angularTarget=<0 0 0>, linearAxesMax=<2 2 2>, linearAxesMin=<-2 -2 -2>}
      localJointOffset=<0 0 -1>
     }]
   }
   n27 = Anim Dynamics - Bone: dyn_l_eye(ComponentPose=@n26, Alpha=0.6, ExternalForce="0,0,0") {
    alpha=0.9
    angularDampingOverride=0.9999
    angularSpringConstant=1000
    bAngularSpring=true
    bGravityOverrideInSimSpace=true
    bOverrideAngularDamping=true
    gravityScale=0.5
    numSolverIterationsPostUpdate=4
    numSolverIterationsPreUpdate=16
    alphaBoolBlend {blendOption=Linear}
    boundBone {boneName=dyn_l_eye}
    componentLinearAccScale=<1 1 1>
    componentLinearVelScale=<1 1 1>
    physicsBodyDefinitions [{
      sphereCollisionRadius=0
      boundBone {boneName=dyn_l_eye}
      boxExtents=<7 5 2>
      constraintSetup {angularTargetAxis=AxisY, twistAxis=AxisZ, angularLimitsMax=<5 10 5>, angularLimitsMin=<-5 -10 -5>, angularTarget=<0 1 0>}
      localJointOffset=<0 0 -1>
     }]
   }
   n28 = Anim Dynamics - Bone: dyn_r_eye(ComponentPose=@n27, Alpha=0.4, ExternalForce="0,0,0") {
    alpha=0.9
    angularDampingOverride=0.9999
    angularSpringConstant=1000
    bAngularSpring=true
    bGravityOverrideInSimSpace=true
    bOverrideAngularDamping=true
    gravityScale=0.5
    numSolverIterationsPostUpdate=4
    numSolverIterationsPreUpdate=16
    alphaBoolBlend {blendOption=Linear}
    boundBone {boneName=dyn_r_eye}
    componentLinearAccScale=<1 1 1>
    componentLinearVelScale=<1 1 1>
    physicsBodyDefinitions [{
      sphereCollisionRadius=0
      boundBone {boneName=dyn_r_eye}
      boxExtents=<7 5 2>
      constraintSetup {angularTargetAxis=AxisY, twistAxis=AxisZ, angularLimitsMax=<5 10 5>, angularLimitsMin=<-5 -10 -5>, angularTarget=<0 1 0>}
      localJointOffset=<0 0 -1>
     }]
   }
   n29 = Anim Dynamics - Bone: dyn_hat(ComponentPose=@n28, Alpha=0.4, ExternalForce="0,0,0") {
    alpha=0.9
    angularDampingOverride=0.9999
    angularSpringConstant=50
    bAngularSpring=true
    bGravityOverrideInSimSpace=true
    bOverrideAngularDamping=true
    gravityScale=0.2
    linearSpringConstant=50
    numSolverIterationsPostUpdate=4
    numSolverIterationsPreUpdate=16
    alphaBoolBlend {blendOption=Linear}
    boundBone {boneName=dyn_hat}
    componentLinearAccScale=<0.25 0.25 0.25>
    physicsBodyDefinitions [{
      sphereCollisionRadius=0
      boundBone {boneName=dyn_hat}
      boxExtents=<10 10 5>
      constraintSetup {angularTargetAxis=AxisY, twistAxis=AxisZ, angularLimitsMax=<25 25 25>, angularLimitsMin=<-25 -25 -25>, angularTarget=<0 1 0>}
      localJointOffset=<0 0 -1>
     }]
    relativeSpaceBone {boneName=root}
   }
   Output Pose(Result=Component To Local(ComponentPose=@n29)) {layerGroup=None}
  ]}
 EventGraph {nodes [
   Event BlueprintUpdateAnimation()
   n1 = GetGameTimeInSeconds()
   n2 = float * float(A=@n1, B=800)
   > Set Jumping_Clap(Jumping_Clap:double=MapRangeClamped(Value="Sin (Degrees)"!(A:double=@n2) -> (ReturnValue:double), InRangeA=-1, InRangeB=1, OutRangeB=0.8)) -> (Output_Get:double)
   n3 = float * float(A=@n1, B=800)
   > Set Sprinting_Clap(Sprinting_Clap:double=MapRangeClamped(Value="Sin (Degrees)"(A=@n3), InRangeA=-1, InRangeB=1, OutRangeB=0.68)) -> (Output_Get:double)
   TryGetPawnOwner()
  ]}
]
variables [
 Jumping_Clap {type=number}
 Sprinting_Clap {type=number}
]
